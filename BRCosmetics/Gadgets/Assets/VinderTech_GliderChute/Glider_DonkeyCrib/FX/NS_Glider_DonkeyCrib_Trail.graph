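# Niagara trail effect for the DonkeyCrib glider, with engine flames, glows with lights, smoke, speed lines and wind cones.
# Flames, glows, smoke and deploy cones attach to sampled skeletal mesh sockets, with partition-indexed arrays placing individual glows.
# System timeline, push, age and startup curves drive deploy bursts and intensity, while user forward and right inputs steer speed lines.
# FrontEnd modifiers swap values between gameplay and front-end display.
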NiagaraSystem /BRCosmetics/Gadgets/Assets/VinderTech_GliderChute/Glider_DonkeyCrib/FX/NS_Glider_DonkeyCrib_Trail
dependencies {
 usedDataInterfaces [
  &NiagaraDataInterfaceCurve
  &NiagaraDataInterfaceVector2DCurve
  &NiagaraDataInterfaceColorCurve
  &NiagaraDataInterfaceArrayInt32
  &NiagaraDataInterfaceArrayFloat3
  &NiagaraDataInterfaceArrayFloat
  &NiagaraDataInterfaceVectorCurve
 ]
 usedDynamicInputs []
 usedModules [
  &self:SystemSpawnScript.SystemScriptSource.SystemScriptGraph.NiagaraNodeAssignment_0.SetVariables_#1
  &SystemState
  &FN_Integrator
  &Timeline
  /BRCosmetics/Characters/Player/Female/Medium/Bodies/F_MED_PiperShelf_Pearl/FX/TimeOfDayGlobal
  &self:SystemSpawnScript.SystemScriptSource.SystemScriptGraph.NiagaraNodeAssignment_1.SetVariables_#2
  &self:SystemSpawnScript.SystemScriptSource.SystemScriptGraph.NiagaraNodeAssignment_3.SetVariables_#3
 ]
}
emitters {
 EngineFlames {
  emitterUpdateScript [
   &EmitterState(Life Cycle Mode=::NewEnumerator1, Loop Behavior=::NewEnumerator1, Loop Delay=$User.EngineDelay, Loop Duration=1, UseLoopDelay=true)
   &SpawnBurst_Instantaneous(Spawn Count=3, Use Spawn Probability=true)
   /SetVariables(
    Emitter.Flame1A:LinearColor=<rgba 0.125 0.0688 0.0125 1>
    Emitter.Flame1B:LinearColor=<rgba 0.5 0.4333 0.1 1>
    Emitter.Flame2A:LinearColor=<rgba 1 0.6667 0.2 1>
    Emitter.Flame2B:LinearColor=<rgba 0.4 0.2133 0.08 1>
    Emitter.Flame3A:LinearColor=<rgba 1 0.4838 0.1 1>
    Emitter.Flame3B:LinearColor=<rgba 0.4583 0.2796 0.0458 1>
    Emitter.FlameCoreA:LinearColor=<rgba 1 0.8461 0.2 1>
    Emitter.FlameCoreB:LinearColor=<rgba 1 0.7 0.1 1>
    Emitter.FlameDark:LinearColor=<rgba 0.1927 0.1172 0.0416 1>
    Emitter.FlameLight:LinearColor=<rgba 0.3854 0.2318 0.0781 1>
   ) {isSetParameters=true}
  ]
  particleSpawnScript [
   &PartitionParticles(ID to Partition:NiagaraInt32=$Engine.ExecIndex, Number Of Partitions=3)
   &FN_SetSpawnGroup(Number of Spawn Groups=3)
   &InitializeParticle(
    Lifetime=0.25
    Mesh Scale=<0.14 0.14 0.24>
    Mesh Scale Mode=::NewEnumerator4
    Sprite Size Mode=::NewEnumerator3
    Uniform Sprite Size=220
    UsePositionOffset=false
    Write Material Random=true
   )
   &InitialMeshOrientation(Orientation Axis=<0 -1 0>, Orientation Vector=<0 0 1>, Use Orientation Vector=true, Use Rotation Vector=false) {module=InitialMeshOrientation001}
   /SetVariables(
    Particles.Flame2Scale:Vector3f=&Multiply_Vector(A=$Particles.Scale, B=<1 1 0.9>)
    Particles.FlameCoreScale:Vector3f=&Multiply_Vector(A=$Particles.Scale, B=<0.65 0.6 0.6>)
   ) {isSetParameters=true}
  ]
  particleUpdateScript [
   &ParticleState(Kill Particles When Lifetime Has Elapsed=false)
   &SkeletalMeshLocation(
    Apply Sampled Orientation to Mesh Particles=true
    Bone / Socket Index=&SelectIntFromArray(Direct Array Index=$Particles.SpawnGroup, Int Selection Array=DataInterface() {propertyValues {IntData [0,1,2]}})
    Inherit Velocity=false
    MaskSpawnGroup=false
    Mesh Sampling Type=::NewEnumerator5
    Sampled Position Offset=&SelectVectorFromArray(Direct Array Index=$Particles.PartitionParticles.Partition, Vector Selection Array=DataInterface() {propertyValues {FloatData []}})
    Sampled Position Offset Coordinate Space:ENiagaraMeshTransforms=::NewEnumerator3
    Skeletal Mesh=$System.Skeletal Mesh
    Socket Sampling Mode=::NewEnumerator5
   )
   &SolveForcesAndVelocity(Clamp Velocity=false, Limit Acceleration=false)
   &ScaleColor(
    Scale Alpha=&FloatFromCurve(FloatCurve=DataInterface() {propertyValues {Curve {keys [<at 0 0 RCIM_Cubic>,<at 0.5 1 RCIM_Cubic>]}}})
    Scale RGB=&MakeVectorFromLinearColorRGB(Color=&LerpLinearColors(StartColor=<rgba 1 1 1 1>, EndColor=<rgba 1.5 1.5 1.5 1>, LerpFactor=$System.Timeline.Time))
    ScaleA=true
    ScaleRGB=true
   )
   /SetVariables(
    Particles.Flame2Color:LinearColor=&MakeLinearColorFromVectorAndFloat("Vector (RGB)"=<1 1 1>, "Float (Alpha)"=&Lerp_Float(A=1, B=5, Alpha=$System.Timeline.Time))
    Particles.FlameCoreColor:LinearColor=&MakeLinearColorFromVectorAndFloat("Vector (RGB)"=&MakeVectorFromLinearColorRGB(Color=<rgba 10 3.6336 0 1>), "Float (Alpha)"=0.3)
   ) {isSetParameters=true}
   &ScaleMeshSize(Scale Factor=&MakeVector(
     X=&FloatFromCurve(FloatCurve=DataInterface() {propertyValues {Curve {keys [{tangentMode=RCTM_User, time=0, value=0}, {tangentMode=RCTM_User, time=0.5, value=1}]}}})
     Y=&FloatFromCurve(FloatCurve=DataInterface() {propertyValues {Curve {keys [{tangentMode=RCTM_User, time=0, value=0}, {tangentMode=RCTM_User, time=0.5, value=1}]}}})
     Z=&FN_FrontEndFloatModifier(
      Gameplay=&FloatFromCurve(FloatCurve=DataInterface() {propertyValues {Curve {keys [<at 0 0>,<at 0.85 1>]}}})
      Front End=$System.StartupCurve
     )
    ))
   &ScaleMeshSize(Scale Factor=&MakeVector(
     X=1
     Y=1
     Z=&Add_Float(
      A=&Lerp_Float(A=&FN_FrontEndFloatModifier(Gameplay=0.8, Front End=1.5), B=2.5, Alpha=$System.Push)
      B=&Multiply_Float(A=1, B=&Max_Float(A:NiagaraFloat=$System.ForwardNoRewCurve, B:NiagaraFloat=$System.StartupBurst))
     )
    )) {module=ScaleMeshSize001}
   /SetVariables(
    Particles.Flame2Scale:Vector3f=&Multiply_Vector(
     A=$Particles.Initial.Flame2Scale
     B=&MakeVector(
      X=&FloatFromCurve(FloatCurve=DataInterface() {propertyValues {Curve {keys [<at 0 0>,<at 1 1>]}}})
      Y=1
      Z=1
     )
    )
    Particles.FlameCoreScale:Vector3f=&Multiply_Vector(
     A=$Particles.Initial.FlameCoreScale
     B=&MakeVector(
      X=&FloatFromCurve(FloatCurve=DataInterface() {propertyValues {Curve {keys [<at 0 0>,<at 1 1>]}}})
      Y=1
      Z=1
     )
    )
   ) {isSetParameters=true}
   /SetVariables(
    Particles.Flame2Scale:Vector3f=&Multiply_Vector(
     A=$Particles.Flame2Scale
     B=&MakeVector(
      X=&FloatFromCurve(FloatCurve=DataInterface() {propertyValues {Curve {keys [{tangentMode=RCTM_User, time=0, value=0}, {tangentMode=RCTM_User, time=0.5, value=1}]}}})
      Y=&FloatFromCurve(FloatCurve=DataInterface() {propertyValues {Curve {keys [{tangentMode=RCTM_User, time=0, value=0}, {tangentMode=RCTM_User, time=0.5, value=1}]}}})
      Z=&Multiply_Float(
       A=&Lerp_Float(
        A=&SelectFloatFromArray(Direct Array Index=$Particles.PartitionParticles.Partition, Float Selection Array=DataInterface() {propertyValues {FloatData [1.2,1.2]}})
        B=&SelectFloatFromArray(Direct Array Index=$Particles.PartitionParticles.Partition, Float Selection Array=DataInterface() {propertyValues {FloatData [4,4]}})
        Alpha=$System.StartupCurve
       )
       B=&Lerp_Float(
        A=&FN_FrontEndFloatModifier(Gameplay=1, Front End=$System.StartupCurve)
        B=2.5
        Alpha=$System.Timeline.Time
       )
      )
     )
    )
    Particles.FlameCoreScale:Vector3f=&Multiply_Vector(
     A=$Particles.FlameCoreScale
     B=&Multiply_Vector(
      A=&MakeVector(
       X=&FloatFromCurve(FloatCurve=DataInterface() {propertyValues {Curve {keys [{tangentMode=RCTM_User, time=0, value=0}, {tangentMode=RCTM_User, time=0.5, value=1}]}}})
       Y=1
       Z=&Lerp_Float(
        A=&SelectFloatFromArray(Direct Array Index=$Particles.PartitionParticles.Partition, Float Selection Array=DataInterface() {propertyValues {FloatData [1,1]}})
        B=&SelectFloatFromArray(Direct Array Index=$Particles.PartitionParticles.Partition, Float Selection Array=DataInterface() {propertyValues {FloatData [1.7,1.7]}})
        Alpha=$System.StartupCurve
       )
      )
      B=&MakeVector(
       X=1
       Y=1
       Z=&Lerp_Float(
        A=&FN_FrontEndFloatModifier(Gameplay=1, Front End=$System.StartupCurve)
        B=3
        Alpha=$System.Timeline.Time
       )
      )
     )
    )
   ) {isSetParameters=true}
   &DynamicMaterialParameters(
    Index 0 Param 1=&Lerp_Float(A=0.3, B=0.6, Alpha=1)
    Module.Param0WriteEnabled=true
    Module.Param1WriteEnabled=true
    Module.Param2WriteEnabled=true
    Module.Param3WriteEnabled=true
   )
  ]
  rendererClasses [&NiagaraMeshRendererProperties]
  renderers [
   {rendererClass=&NiagaraMeshRendererProperties, rendererIndex=0}
   {rendererClass=&NiagaraMeshRendererProperties, rendererIndex=1}
   {rendererClass=&NiagaraMeshRendererProperties, rendererIndex=2}
   {rendererClass=&NiagaraMeshRendererProperties, rendererIndex=3}
  ]
 }
 Glow {
  emitterUpdateScript [
   &EmitterState(
    Inactive Response=::NewEnumerator1
    Life Cycle Mode=::NewEnumerator1
    Loop Behavior=::NewEnumerator1
    Loop Duration Mode=::NewEnumerator1
    UseLoopDelay=false
   )
   &SpawnBurst_Instantaneous(
    Spawn Count=3
    Spawn Time=&FN_FrontEndFloatModifier(Gameplay=0, Front End=0.1)
    Use Spawn Probability=false
    UseLoopCountLimit=false
   )
   &SpawnBurst_Instantaneous(
    Spawn Count=22
    Spawn Group=1
    Spawn Time=&FN_FrontEndFloatModifier(Gameplay=0, Front End=0.1)
    Use Spawn Probability=false
    UseLoopCountLimit=false
   ) {module=SpawnBurst_Instantaneous001}
  ]
  particleSpawnScript [
   &PartitionParticles(Number Of Partitions=25)
   /SetVariables(Particles.IsRingGlow:NiagaraFloat=&MakeCustomFloatFromBool(
     Bool=&SetBoolByFloatComparison(A=&MakeFloatFromInt(Int=$Particles.PartitionParticles.Partition), B=8)
     True Float=1
     False Float=0
    )) {isSetParameters=true}
   &InitializeParticle(
    Position Offset=&SelectVectorFromArray(Direct Array Index=&SpawnGroup(), Vector Selection Array=DataInterface() {propertyValues {FloatData []}})
    Sprite Size Mode=::NewEnumerator3
    Uniform Sprite Size=25
    UsePositionOffset=true
    Write Material Random=true
   )
  ]
  particleUpdateScript [
   /SetVariables(Particles.Jitter:NiagaraFloat=&Waveform(
     Global Input Position=&RandomRangeFloat(Minimum=0, Maximum=1)
     "Waveform [1]"=::NewEnumerator12
     "[1] Frequency"=10
     "[1] Amplitude Min/Max"=<0.75 1>
    )) {isSetParameters=true}
   &ParticleState(Kill Particles When Lifetime Has Elapsed=false)
   &ScaleSpriteSize(
    Scale Sprite Size Mode=::NewEnumerator0
    Uniform Scale Factor=&Add_Float(
     A=1
     B=&Multiply_Float(A=0.3, B=&Add_Float(A=$System.ForwardNoRewCurve, B=$System.StartupBurst))
    )
   )
   &SkeletalMeshLocation(
    Bone / Socket Index=&SelectIntFromArray(Direct Array Index=$Particles.PartitionParticles.Partition, Int Selection Array=DataInterface() {propertyValues {IntData [0,1,2,3]}})
    MaskSpawnGroup=false
    Mesh Sampling Type=::NewEnumerator5
    Sampled Position Offset=&SelectVectorFromArray(
     Direct Array Index=$Particles.PartitionParticles.Partition
     Vector Selection Array=DataInterface() {propertyValues {FloatData [
        <0 0 0>
        <0 0 0>
        <0 0 0>
        <-10 -15 1.5>
        <-6 -15 -6>
        <6 -15 -6>
        <10 -15 1.5>
        <4.3 -15 12>
        <-4.3 -15 12>
        <-52.5 30 -5>
        <-52 30 5.2>
        <-49.5 30 15.5>
        <-45 30 24.5>
        <-38.5 30 33>
        <-31.5 30 39.5>
        <-22.2 30 45>
        <-12.5 30 48.4>
        <52.5 30 -5>
        <52 30 5.2>
        <49.5 30 15.5>
        <45 30 24.5>
        <38.5 30 33>
        <31.5 30 39.5>
        <22.2 30 45>
        <12.5 30 48.4>
       ]}}
    )
    Skeletal Mesh=$System.Skeletal Mesh
    Socket Sampling Mode=::NewEnumerator5
   )
   &ScaleColor(
    Scale Alpha=&Multiply_Float(
     A=&Multiply_Float(A=$Particles.Jitter, B=&FN_FrontEndFloatModifier(Gameplay=0.5, Front End=0.25))
     B=$System.EmissiveFadeInCurve
    )
    Scale RGB=&Multiply_VectorByFloat(
     Vector=&Multiply_VectorByFloat(
      Vector=&MakeVectorFromLinearColorRGB(Color=<rgba 1 0.6458 0.15 1>)
      Float=&Lerp_Float(A=0.3, B=0.7, Alpha=$System.Push)
     )
     Float=&Add_Float(
      A=1
      B=&Multiply_Float(A=0.5, B=&Max_Float(A:NiagaraFloat=$System.ForwardNoRewCurve, B:NiagaraFloat=$System.StartupBurst))
     )
    )
    ScaleA=true
    ScaleRGB=true
   )
   &FN_Light_Attributes(
    Color=$Particles.Color
    Create Color Attribute=true
    Create Exponent Attribute=true
    Create Position Attribute=true
    Create Radius Attribute=true
    Exponent=3
    Position=&AddVectorToPosition(
     Position=$Particles.SkeletalMeshLocation.SampledPosition
     Vector=&SelectVectorFromArray(
      Direct Array Index=$Particles.PartitionParticles.Partition
      Vector Selection Array=DataInterface() {propertyValues {FloatData [<0 20 0>, <0 20 0>]}}
     )
    )
    Radius=&Lerp_Float(A=40, B=60, Alpha=$System.Push)
    Scale Color=&Multiply_Float(A=$Particles.Jitter, B=&Lerp_Float(A=30, B=70, Alpha=$System.Push))
   )
   &FN_Sprite_Attributes(
    Camera Offset:NiagaraFloat=&FN_FrontEndFloatModifier(Gameplay=20, Front End=5)
    Color=&Multiply_LinearColorByFloat(
     Linear Color=<rgba 1 0.6458 0.15 1>
     Float=&Add_Float(
      A=&Lerp_Float(A=0.7, B=1, Alpha=$System.Push)
      B=&Multiply_Float(
       A=&FN_FrontEndFloatModifier(Gameplay=1, Front End=0.5)
       B=&Max_Float(A:NiagaraFloat=$System.ForwardNoRewCurve, B:NiagaraFloat=$System.StartupBurst)
      )
     )
    )
    Create Camera Offset Attribute=true
    Create Color Attribute=true
    Create Position Attribute=true
    Create SpriteSize Attribute=true
    Position:NiagaraPosition=$Particles.SkeletalMeshLocation.SampledPosition
    SpriteSize=&Multiply_Vector2D_ByFloat(
     Float=&Add_Float(
      A=1
      B=&Multiply_Float(A=0.5, B=&Max_Float(A:NiagaraFloat=$System.ForwardNoRewCurve, B:NiagaraFloat=$System.StartupBurst))
     )
     Vector2D=&FN_FrontEndVector2Modifier(Front End:Vector2f=<40 40>, Gameplay:Vector2f=<50 50>)
    )
    Write to Module Namespace=true
   ) {module=FN_Sprite_Attributes001}
   &CameraOffset(Camera Offset Amount=20)
   /SetVariables(
    Particles.BlueCameraOffset=8
    Particles.BlueColor:LinearColor=&MakeLinearColorFromVectorAndFloat(
     "Vector (RGB)"=&MakeVectorFromLinearColorRGB(Color=<rgba 0.1 0.6009 1 1>)
     "Float (Alpha)"=&Multiply_Float(
      A=&Multiply_Float(
       A=$Particles.Jitter
       B=&Lerp_Float(
        A=&FN_FrontEndFloatModifier(Gameplay=0.75, Front End=0.35)
        B=&Add_Float(
         A=&Multiply_Float(
          A=0.15
          B=&FloatFromCurve(
           FloatCurve=DataInterface() {propertyValues {Curve {keys [<at 0.8 0>,<at 1 1>]}}}
           CurveIndex=&Add_Float(A=$System.Age, B=&FN_FrontEndFloatModifier(Gameplay=0, Front End=-0.2))
           Scale Curve=&FN_FrontEndFloatModifier(Gameplay=1, Front End=0.5)
          )
         )
         B=&Multiply_Float(A=0.1, B=$System.StartupBurst)
        )
        Alpha=$Particles.IsRingGlow
       )
      )
      B=$System.EmissiveFadeInCurve
     )
    )
    Particles.BlueSpriteSize:Vector2f=&Multiply_Vector2D_ByFloat(
     Float=&Add_Float(
      A=1
      B=&Lerp_Float(A=0, B=&Multiply_Float(A=0.25, B=$System.StartupBurst), Alpha=$Particles.IsRingGlow)
     )
     Vector2D=<15 15>
    )
    Particles.VisibilityTag=&SpawnGroup()
   ) {isSetParameters=true}
  ]
  rendererClasses [&NiagaraSpriteRendererProperties, &NiagaraLightRendererProperties]
  renderers [
   {rendererClass=&NiagaraSpriteRendererProperties}
   {rendererClass=&NiagaraLightRendererProperties, rendererIndex=1}
   {rendererClass=&NiagaraSpriteRendererProperties, rendererIndex=2}
   {rendererClass=&NiagaraSpriteRendererProperties, rendererIndex=3}
  ]
 }
 Smoke {
  emitterUpdateScript [
   &EmitterState()
   &SpawnRate(
    SpawnRate=&Lerp_Float(
     A=&Lerp_Float(
      A=30
      B=50
      Alpha=&FloatFromCurve(FloatCurve=DataInterface() {propertyValues {Curve {keys [<at 0 0>,<at 0.2 1>]}}}, CurveIndex=$System.Timeline.NormalizedTime)
     )
     B=0
     Alpha=&FloatFromCurve(FloatCurve=DataInterface() {propertyValues {Curve {keys [<at 0.8 1>,<at 1.2 0>]}}}, CurveIndex=$System.Age)
    )
    Use Spawn Probability=true
   )
   &SpawnRate(
    SpawnGroup=1
    SpawnRate=&FN_FrontEndFloatModifier(
     Gameplay=&FloatFromCurve(
      FloatCurve=DataInterface() {propertyValues {Curve {keys [
          <at 0.4 0>
          {arriveTangent=2.5, leaveTangent=2.5, time=0.45, value=1}
          {arriveTangent=-0.9524, leaveTangent=-0.9524, time=1, value=1}
          <at 1.2 0>
         ]}}}
      CurveIndex=$System.Age
      Scale Curve=120
     )
     Front End=&FloatFromCurve(
      FloatCurve=DataInterface() {propertyValues {Curve {keys [{arriveTangent=2.5, time=0.45, value=1}, {arriveTangent=-0.9524, leaveTangent=-0.9524, time=1, value=1}, <at 1.2 0>]}}}
      CurveIndex=$System.Age
      Scale Curve=120
     )
    )
    Use Spawn Probability=true
   ) {module=SpawnRate001}
  ]
  particleSpawnScript [
   &PartitionParticles(Number Of Partitions=3)
   &InitializeParticle(
    Color=&MultiplyLinearColor(
     Linear Color=<rgba 1 1 1 1>
     Scale Factor=&MakeVector4_FromFloat(Float=&FN_FrontEndFloatModifier(Gameplay=1, Front End=0.5))
    )
    Lifetime Max=0.7
    Lifetime Min=0.5
    Lifetime Mode=::NewEnumerator1
    Mesh Scale=&Multiply_VectorByFloat(Vector=&RandomRangeVector(Minimum=<4 1.5 1.5>, Maximum=<5 2.5 2.5>), Float=0.22)
    Mesh Scale Mode=::NewEnumerator4
    UsePositionOffset=false
    Write Material Random=true
   )
   &AddVelocity(
    Coordinate Space=::Simulation
    Scale Added Velocity=&MergeHelper_VectorSwitch(
     0=&Multiply_VectorByFloat(Vector=<2.5 2.5 2.5>, Float=&Lerp_Float(A=0.3, B=1, Alpha=$System.Timeline.NormalizedTime))
     1=&VectorFromFloat(Value=&FloatFromCurve(
       FloatCurve=DataInterface() {propertyValues {Curve {keys [{interpMode=RCIM_Cubic, tangentMode=RCTM_User, time=0.6, value=1}, {interpMode=RCIM_Cubic, tangentMode=RCTM_User, time=1.2, value=0}]}}}
       CurveIndex=$System.Age
       Scale Curve=&FN_FrontEndFloatModifier(Gameplay=1, Front End=3)
      ))
     2=?
     3=?
     Index=&SpawnGroup()
     Num=::NewEnumerator0
    )
    Velocity=&MergeHelper_VectorSwitch(0=<0 100 0>, 1=<0 300 0>, 2=?, 3=?, Index=&SpawnGroup(), Num=::NewEnumerator0)
   )
   &SkeletalMeshLocation(
    Bone / Socket Index=&SelectIntFromArray(Direct Array Index=$Particles.PartitionParticles.Partition, Int Selection Array=DataInterface() {propertyValues {IntData [0,1,2]}})
    "Inherit Velocity (Scale)":NiagaraFloat=&FN_FrontEndFloatModifier(
     Gameplay=&FloatFromCurve(FloatCurve=DataInterface() {propertyValues {Curve {keys [<at 0.7 1>,<at 0.9 0>]}}}, CurveIndex=$System.Age)
     Front End=0
    )
    MaskSpawnGroup=false
    Mesh Sampling Type=::NewEnumerator5
    Orientation Sampling=::NewEnumerator1
    Skeletal Mesh=$System.Skeletal Mesh
    Socket Sampling Mode=::NewEnumerator5
    Velocity Sampling=::NewEnumerator0
    VelocitySpeedLimit=false
   )
   &ShapeLocation(MaskSpawnGroup=false, Override Seed=false, Sphere Radius=1, UseOwnerScale=true)
   &AddVelocityInCone(
    Cone Angle=50
    Cone Axis=<0 1 0>
    Cone Axis Coordinate Space=::Simulation
    Module.Use Velocity Falloff On Cone Axis=false
    Velocity Strength=66.6
   ) {module=AddVelocityInCone001}
   &InitialMeshOrientation(
    Mesh Orientation Mode=::NewEnumerator2
    Rotation=&RandomRangeVector(Minimum=<0 0 0>, Maximum=<1 0 0>)
    Use Rotation Vector=true
   )
   &CameraOffset(Camera Offset Amount=-10)
  ]
  particleUpdateScript [
   &Drag(Use Linear Drag=true, Use Rotational Drag=true)
   &SolveForcesAndVelocity(Clamp Velocity=false, Limit Acceleration=false)
   &ParticleState()
   &ScaleMeshSize(Scale Factor=&Multiply_VectorByFloat(
     Vector=&Multiply_VectorByFloat(
      Vector=&Multiply_Vector(
       A=<1 1 1>
       B=&VectorFromCurve(VectorCurve=DataInterface() {propertyValues {XCurve {keys [<at 0 5>,<at 0.2 1>,<at 1 1>]}, YCurve {keys [<at 0 1>,<at 1 1>]}, ZCurve {keys [<at 0 1>,<at 1 1>]}}})
      )
      Float=&RandomRangeFloat(Minimum=1, Maximum=1.5)
     )
     Float=&FloatFromCurve(
      FloatCurve=DataInterface() {propertyValues {Curve {keys [
          {interpMode=RCIM_Cubic, tangentMode=RCTM_User, time=0, value=0.2}
          {arriveTangent=1.3433, tangentMode=RCTM_User, time=0.2, value=0.5}
          {arriveTangent=1.4526, interpMode=RCIM_Cubic, leaveTangent=1.4526, tangentMode=RCTM_User, time=1, value=1.5}
         ]}}}
      Scale Curve=&Multiply_Float(
       A=0.83
       B=&MergeHelper_FloatSwitch(
        0=&Lerp_Float(A=0.5, B=1, Alpha=$System.Timeline.NormalizedTime)
        1=1
        2=?
        3=?
        Index=&SpawnGroup()
        Num=::NewEnumerator0
       )
      )
     )
    ))
   &DynamicMaterialParameters(
    Index 0 Param 1=&FloatFromCurve(
     FloatCurve=DataInterface() {propertyValues {Curve {keys [
         {interpMode=RCIM_Cubic, tangentMode=RCTM_User, time=0, value=0}
         <at 0.2 0.55>
         {arriveTangent=-1.0223, interpMode=RCIM_Cubic, leaveTangent=-1.0223, tangentMode=RCTM_User, time=1, value=1}
        ]}}}
     Scale Curve=&RandomRangeFloat(Minimum=0.7, Maximum=1.5)
    )
    Index 0 Param 4=&Lerp_Float(
     A=0.6
     B=0
     Alpha=&FloatFromCurve(FloatCurve=DataInterface() {propertyValues {Curve {keys [
          {arriveTangent=0.7438, interpMode=RCIM_Cubic, leaveTangent=0.7438, tangentMode=RCTM_User, time=0, value=0}
          {arriveTangent=0.0023, interpMode=RCIM_Cubic, leaveTangent=0.0023, tangentMode=RCTM_User, time=1, value=1}
         ]}}})
    )
    Module.Index0Param0Write=true
    Module.Index0Param1Write=true
    Module.Index0Param2Write=true
    Module.Index0Param3Write=true
   ) {module=DynamicMaterialParameters001}
   &ScaleColor(
    Scale Alpha=&FloatFromCurve(
     FloatCurve=DataInterface() {propertyValues {Curve {keys [
         {arriveTangent=2.8927, interpMode=RCIM_Cubic, leaveTangent=2.8927, tangentMode=RCTM_User, time=0, value=0}
         <at 0.05 1>
         {interpMode=RCIM_Cubic, tangentMode=RCTM_Break, time=0.5, value=1}
         <at 1 0>
        ]}}}
     Scale Curve=&MergeHelper_FloatSwitch(
      0=&Lerp_Float(A=0.15, B=0.3, Alpha=$System.Timeline.NormalizedTime)
      1=0.2
      2=?
      3=?
      Index=&SpawnGroup()
      Num=::NewEnumerator0
     )
    )
    Scale RGB=&MakeVectorFromLinearColorRGB(Color=&ColorFromCurve(ColorCurve=DataInterface() {propertyValues {
        AlphaCurve {keys []}
        BlueCurve {keys [<at 0.0108 0 RCIM_Cubic>,<at 0.2563 0.4 RCIM_Cubic>]}
        GreenCurve {keys [<at 0.0108 26.6667 RCIM_Cubic>,<at 0.2563 0.9333 RCIM_Cubic>]}
        RedCurve {keys [<at 0.0108 40 RCIM_Cubic>,<at 0.2563 2 RCIM_Cubic>]}
       }}))
    ScaleA=true
    ScaleRGB=true
   )
  ]
  rendererClasses [&NiagaraMeshRendererProperties]
  renderers [{rendererClass=&NiagaraMeshRendererProperties, rendererIndex=0}]
 }
 Speedlines001 {
  emitterUpdateScript [
   &EmitterState()
   &SpawnRate(
    Spawn Probability=&MakeCustomFloatFromBool(
     Bool=&SetBoolByFloatComparison(A=$User.ForwardDot, B=0.05)
     True Float=1
     False Float=&FN_FrontEndFloatModifier(Gameplay=0, Front End=1)
    )
    SpawnRate=&MakeCustomFloatFromBool(Bool=$User.bIsFrontEndPreview, True Float=50, False Float=90)
    Use Spawn Probability=true
   ) {module=SpawnRate001}
  ]
  particleSpawnScript [
   &InitializeParticle(
    Color=&MakeLinearColorFromVectorAndFloat(
     "Vector (RGB)"=<1 1 1>
     "Float (Alpha)"=&Multiply_Float(
      A=&UniformRangedFloat(Minimum=1, Maximum=1)
      B=&MakeCustomFloatFromBool(
       Bool=$User.bIsFrontEndPreview
       True Float=&FloatFromCurve(
        FloatCurve=DataInterface() {propertyValues {Curve {keys [{tangentMode=RCTM_User, time=0, value=0}, {tangentMode=RCTM_User, time=2, value=1}]}}}
        CurveIndex=$Emitter.Age
        Scale Curve=0.4
       )
       False Float=1
      )
     )
    )
    Lifetime Max=0.2
    Lifetime Min=0.15
    Lifetime Mode=::NewEnumerator1
    Position Offset=&FN_FrontEndVector3Modifier(Gameplay=<0 -50 0>, Front End=<0 50 0>)
    Sprite Size Max=<3 10>
    Sprite Size Min=<2 2>
    Sprite Size Mode=::NewEnumerator2
    UsePositionOffset=true
    Write Material Random=true
   )
   &AddVelocity(
    Module.Use Falloff Distance=false
    Override Seed=false
    Velocity Mode=::NewEnumerator1
    Velocity Speed=&FN_FrontEndFloatModifier(Gameplay=-2500, Front End=2500)
   ) {module=AddVelocity001}
   &TorusLocation(
    Align Torus With Axis=true
    Coordinate Space:ENiagaraCoordinateSpace=::Simulation
    Handle Radius=0
    Large Radius:NiagaraFloat=&FN_FrontEndFloatModifier(
     Gameplay=&RandomRangeFloat(Minimum=450, Maximum=500)
     Front End=&RandomRangeFloat(Minimum=350, Maximum=400)
    )
    MaskSpawnGroup=false
    Module.Override Randomness=false
    Offset=<0 33 0>
    Torus Axis=<0 1 0>
    Torus Origin:NiagaraPosition=&AddVectorToPosition(
     Position=$Particles.Position
     Vector=&FN_FrontEndVector3Modifier(Gameplay=<0 -500 0>, Front End=<0 -500 0>)
    )
   )
   &AddVelocity(
    Velocity=&Add_Vector(
     A=&RandomRangeVector(Minimum=<0 300 0>, Maximum=<0 700 1>)
     B=&Add_Vector(
      A=&Multiply_VectorByFloat(Vector=<-555 0 0>, Float=$User.RightDot)
      B=&Multiply_VectorByFloat(Vector=<0 0 222>, Float=$User.ForwardDot)
     )
    )
    Velocity Speed Scale=&FN_FrontEndFloatModifier(Gameplay=1, Front End=-1)
   )
  ]
  particleUpdateScript [
   &SolveForcesAndVelocity(Clamp Velocity=false, Limit Acceleration=false)
   &ParticleState()
   &ScaleSpriteSize(Scale Factor=&Vector2DFromCurve(
     Vector2Curve=DataInterface() {propertyValues {XCurve {keys [<at 0 0.5 RCIM_Cubic>,<at 0.5 1 RCIM_Cubic>]}, YCurve {keys [<at 0 0 RCIM_Cubic>,<at 0.5 1 RCIM_Cubic>]}}}
     ScaleCurve=&Vector2DFromFloat(Value=1)
    ))
   &ScaleSpriteSizeBySpeed(Max Scale Factor=<1 50>, Min Scale Factor=<0.01 0.05>, Velocity Threshold=1)
   &ScaleColor(
    Scale Alpha=&FloatFromCurve(
     FloatCurve=DataInterface() {propertyValues {Curve {keys [
         <at 0 0 RCIM_Cubic>
         <at 0.3759 1.0207>
         {arriveTangent=-2.221, interpMode=RCIM_Cubic, leaveTangent=-2.221, tangentMode=RCTM_User, time=1, value=0}
        ]}}}
     Scale Curve=&MakeCustomFloatFromBool(
      Bool=$User.bIsFrontEndPreview
      True Float=0.3759
      False Float=&Lerp_Float(A=0, B=0.5, Alpha=&FN_FrontEndFloatModifier(Gameplay=$System.Timeline.Time, Front End=0.5))
     )
    )
    Scale RGB=&VectorFromFloat(Value=&FN_FrontEndFloatModifier(Gameplay=0.5, Front End=0.33))
    ScaleA=true
    ScaleRGB=true
   )
  ]
  rendererClasses [&NiagaraSpriteRendererProperties]
  renderers [{rendererClass=&NiagaraSpriteRendererProperties}]
 }
 WindConeDeploy {
  emitterUpdateScript [
   &EmitterState(Life Cycle Mode=::NewEnumerator1, Loop Duration=1, UseLoopDelay=false)
   &SpawnRate(
    SpawnRate=&FN_FrontEndFloatModifier(
     Gameplay=&Lerp_Float(
      A=0
      B=32
      Alpha=&FloatFromCurve(
       FloatCurve=DataInterface() {propertyValues {Curve {keys [{interpMode=RCIM_Cubic, tangentMode=RCTM_User, time=0.6, value=1}, {interpMode=RCIM_Cubic, tangentMode=RCTM_User, time=0.7, value=0}]}}}
       CurveIndex=$System.Age
      )
     )
     Front End=0
    )
    Use Spawn Probability=false
   )
  ]
  particleSpawnScript [
   &PartitionParticles(Number Of Partitions=3)
   &InitializeParticle(
    Color=&MakeLinearColorFromVectorAndFloat(
     "Vector (RGB)"=&VectorFromFloat(Value=0.7)
     "Float (Alpha)"=&Multiply_Float(A=&RandomRangeFloat(Minimum=0.5, Maximum=1), B=15)
    )
    Lifetime=0.5
    Mesh Scale=&Multiply_VectorByFloat(Vector=<1.25 1 1.3>, Float=0.8)
    Mesh Scale Mode=::NewEnumerator4
    Sprite Size Mode=::NewEnumerator3
    UsePositionOffset=true
    Write Material Random=true
   )
   &InitialMeshOrientation(
    Orientation Axis=<0 1 0>
    Rotation=&FN_FrontEndVector3Modifier(Gameplay=<0 0 0>, Front End=<0.35 0 0>)
    Use Orientation Vector=true
    Use Rotation Vector=false
   )
  ]
  particleUpdateScript [
   &ParticleState()
   &SkeletalMeshLocation(
    Apply Sampled Orientation to Mesh Particles=true
    Bone / Socket Index=$Particles.PartitionParticles.Partition
    Inherit Velocity=false
    MaskSpawnGroup=false
    Mesh Sampling Type=::NewEnumerator5
    Sampled Position Offset=&AddVectorToPosition(
     Position=&MergeHelper_VectorSwitch(0=<3 -100 0>, 1=<-4 -100 0>, 2=<0 -100 -5>, 3=?, Index=$Particles.PartitionParticles.Partition, Num=::NewEnumerator1)
     Vector=&Multiply_VectorByFloat(
      Vector=&Lerp_Vector(
       A=<0 0 0>
       B=<0 40 0>
       Alpha=&FloatFromCurve(FloatCurve=DataInterface() {propertyValues {Curve {keys [<at 0.2 1>,<at 0.8 0>]}}}, CurveIndex=$System.Age)
      )
      Float=&MergeHelper_FloatSwitch(0=1, 1=1, 2=0.4, 3=?, Index=$Particles.PartitionParticles.Partition, Num=::NewEnumerator1)
     )
    )
    Sampled Position Offset Coordinate Space:ENiagaraMeshTransforms=::NewEnumerator3
    Skeletal Mesh=$System.Skeletal Mesh
    Socket Sampling Mode=::NewEnumerator5
   )
   &ScaleMeshSize(Scale Factor=&Multiply_Vector(
     A=<0.8 0.2 0.2>
     B=&Lerp_Vector(
      A=&VectorFromFloat(Value=&FloatFromCurve(FloatCurve=DataInterface() {propertyValues {Curve {keys [<at 0.2 0>,<at 0.6 1>]}}}, CurveIndex=$System.Age))
      B=<1 1 1>
      Alpha=&MergeHelper_FloatSwitch(0=0, 1=0, 2=1, 3=?, Index=$Particles.PartitionParticles.Partition, Num=::NewEnumerator1)
     )
    ))
   &ScaleColor(
    Scale Alpha=&FloatFromCurve(FloatCurve=DataInterface() {propertyValues {Curve {
        defaultValue=1
        keys [
         {arriveTangent=9.0725, interpMode=RCIM_Cubic, leaveTangent=9.0725, tangentMode=RCTM_User, time=0, value=0}
         {arriveTangent=-2.5249, interpMode=RCIM_Cubic, leaveTangent=-2.5249, tangentMode=RCTM_User, time=0.1408, value=0.9847}
         {arriveTangent=-0.0405, interpMode=RCIM_Cubic, leaveTangent=-0.0405, tangentMode=RCTM_User, time=1, value=0}
        ]
       }}})
    ScaleA=true
    ScaleRGB=true
   )
  ]
  rendererClasses [&NiagaraMeshRendererProperties]
  renderers [{rendererClass=&NiagaraMeshRendererProperties, rendererIndex=0}]
 }
 WindConeMain {
  emitterUpdateScript [
   &EmitterState(Life Cycle Mode=::NewEnumerator1, Loop Duration=1, UseLoopDelay=false)
   &SpawnRate(
    SpawnRate=&FN_FrontEndFloatModifier(
     Gameplay=&Max_Float(
      A:NiagaraFloat=&Lerp_Float(
       A=0
       B=16
       Alpha=&FloatFromCurve(
        FloatCurve=DataInterface() {propertyValues {Curve {keys [{interpMode=RCIM_Cubic, tangentMode=RCTM_User, time=0.8, value=1}, {interpMode=RCIM_Cubic, tangentMode=RCTM_User, time=1.1, value=0}]}}}
        CurveIndex=$System.Age
       )
      )
      B:NiagaraFloat=&Lerp_Float(A=0, B=16, Alpha=$User.ForwardDot)
     )
     Front End=0
    )
    Use Spawn Probability=false
   )
  ]
  particleSpawnScript [
   &PartitionParticles()
   &InitializeParticle(
    Color=&MakeLinearColorFromVectorAndFloat(
     "Vector (RGB)"=&VectorFromFloat(Value=0.9)
     "Float (Alpha)"=&Multiply_Float(A=&RandomRangeFloat(Minimum=0.5, Maximum=1), B=8)
    )
    Lifetime=0.5
    Mesh Scale=&Multiply_VectorByFloat(Vector=<1.25 1 1.3>, Float=0.8)
    Mesh Scale Mode=::NewEnumerator4
    Position Offset=&MergeHelper_VectorSwitch(0=<0 -170 0>, 1=<0 -130 0>, 2=?, 3=?, Index=$Particles.PartitionParticles.Partition, Num=::NewEnumerator0)
    Sprite Size Mode=::NewEnumerator3
    UsePositionOffset=true
    Write Material Random=true
   )
   &InitialMeshOrientation(
    Orientation Axis=<0 1 0>
    Rotation=&FN_FrontEndVector3Modifier(Gameplay=<0 0 0>, Front End=<0.35 0 0>)
    Use Orientation Vector=true
    Use Rotation Vector=false
   )
  ]
  particleUpdateScript [
   &ParticleState()
   &ScaleMeshSize(Scale Factor=&MergeHelper_VectorSwitch(0=<1 0.3 0.3>, 1=<0.8 0.5 0.5>, 2=?, 3=?, Index=$Particles.PartitionParticles.Partition, Num=::NewEnumerator0))
   &ScaleColor(
    Scale Alpha=&FloatFromCurve(FloatCurve=DataInterface() {propertyValues {Curve {
        defaultValue=1
        keys [
         {arriveTangent=9.0725, interpMode=RCIM_Cubic, leaveTangent=9.0725, tangentMode=RCTM_User, time=0, value=0}
         {arriveTangent=-2.5249, interpMode=RCIM_Cubic, leaveTangent=-2.5249, tangentMode=RCTM_User, time=0.1408, value=0.9847}
         {arriveTangent=-0.0405, interpMode=RCIM_Cubic, leaveTangent=-0.0405, tangentMode=RCTM_User, time=1, value=0}
        ]
       }}})
    ScaleA=true
    ScaleRGB=true
   )
  ]
  rendererClasses [&NiagaraMeshRendererProperties]
  renderers [{rendererClass=&NiagaraMeshRendererProperties, rendererIndex=0}]
 }
}
summary {
 systemName=NS_Glider_DonkeyCrib_Trail
 emitters [
  {emitterName=Speedlines001, rendererClasses [&NiagaraSpriteRendererProperties]}
  {emitterName=EngineFlames, rendererClasses [&NiagaraMeshRendererProperties]}
  {emitterName=WindConeMain, rendererClasses [&NiagaraMeshRendererProperties]}
  {emitterName=Smoke, rendererClasses [&NiagaraMeshRendererProperties]}
  {emitterName=WindConeDeploy, rendererClasses [&NiagaraMeshRendererProperties]}
  {emitterName=Glow, rendererClasses [&NiagaraSpriteRendererProperties, &NiagaraLightRendererProperties]}
 ]
 userVariables [
  User.bIsFrontEnd {type=NiagaraBool}
  User.bIsFrontEndPreview {type=NiagaraBool}
  User.bIsFullyDeployed {type=NiagaraBool}
  User.EngineDelay {type=NiagaraFloat, defaultValue=0.16, description=""}
  User.ForwardDot {type=NiagaraFloat, defaultValue=0, description=""}
  User.RibbonIDOffset {type=NiagaraInt32}
  User.RightDot {type=NiagaraFloat, defaultValue=0, description=""}
  User.SkeletalMesh {
   type=NiagaraDataInterfaceSkeletalMesh
   defaultValue {
    struct=NiagaraExt_VariableValue_DataInterface
    value {dataInterface=&self:NiagaraDataInterfaceSkeletalMesh_0}
   }
  }
  User.Variant {type=NiagaraFloat, defaultValue=0, description=""}
 ]
}
